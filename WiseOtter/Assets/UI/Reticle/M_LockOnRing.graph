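# Additive UI material for a lock-on reticle ring, drawing a tinted grayscale texture with radial and circular progress masks.
# Center-scaled UI UVs are offset by the scaled, clamped X/Y of CameraVelocity, so the ring drifts slightly with camera motion.
# The sample's opacity is temporarily boosted until final visuals, then multiplied by the radial progress mask, a circle mask sized by CircleMaskProgress, and Color.
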Material /WiseOtter/Assets/UI/Reticle/M_LockOnRing
graph {
 properties {
  bAutomaticallySetUsageInEditor=false
  blendMode=BLEND_Additive
  materialDomain=MD_UI
  pixelDepthOffsetMode=PDOM_Legacy
  usedShadingModels=Default Lit
 }
 nodes [
  n1 = MaterialFunctionCall(
   UVs=MaterialFunctionCall() {materialFunction=&GetUserInterfaceUV}.Normalized UV
   Texture Scale=ScalarParameter() {defaultValue=0.8, parameterName=TextureScale}
  ) {materialFunction=&ScaleUVsByCenter}
  n2 = VectorParameter() {parameterName=CameraVelocity}
  n3 = Constant2Vector() {desc=Max Drift, g=0.2, r=0.2}
  n4 = Clamp(None=Divide(A=ComponentMask(None=@n2.RGB) {g=true, r=true}, B=Constant2Vector() {g=0.84, r=9.6}), Min=Multiply(A=@n3) {constB=-1}, Max=@n3)
  n5 = TextureSample(UVs=NamedRerouteDeclaration(None=Add(A=@n1.UVs, B=@n4)) {name=ParallaxUV, nodeColor=<rgba 0 0 0 1>}) {samplerSource=SSM_Clamp_WorldGroupSettings, samplerType=SAMPLERTYPE_Grayscale, texture=/WiseOtter/Assets/UI/Reticle/Textures/T_LockOn}
  n6 = Saturate(None=Multiply(A=@n5.R) {constB=3, desc="Making this less transparent\r\nuntil we redo with final visuals"})
  n7 = MaterialFunctionCall!(UVs, Progress=ScalarParameter() {defaultValue=1, parameterName=RadialMaskProgress}, StartValue, Range) {materialFunction=/FortUILibrary/MaterialFunctions/Basics/MF_UI_RadialProgress}
  n8 = NamedRerouteUsage()
  n9 = LinearInterpolate(Alpha=ScalarParameter() {parameterName=CircleMaskProgress}) {constA=0.8, constB=0.93}
  Multiply(
   A=VectorParameter() {parameterName=Color, defaultValue=<rgba 0.4062 0.4841 1 1>}.RGB
   B=Multiply(
    A=Multiply(A=@n6, B=@n7.Result)
    B=MaterialFunctionCall(UVs=@n8, Size=@n9) {materialFunction=&MF_UI_Circle}.Fill
   )
  )
 ]
}
