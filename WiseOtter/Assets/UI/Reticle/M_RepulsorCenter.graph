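# Additive UI material for the center of the repulsor reticle.
# Two MF_UI_Circle calls draw a thin outlined circle and a small filled dot, merged with Max into an InnerReticle mask.
# Named reroutes feed a tint-and-Max chain dimmed by a final multiply.
# The Locked? parameter blends two inputs from the same red constant, so it has no visible effect.
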Material /WiseOtter/Assets/UI/Reticle/M_RepulsorCenter
graph {
 properties {
  bAutomaticallySetUsageInEditor=false
  blendMode=BLEND_Additive
  materialDomain=MD_UI
  pixelDepthOffsetMode=PDOM_Legacy
  usedShadingModels=Default Lit
 }
 nodes [
  n1 = Constant4Vector() {constant=<rgba 1 0 0 1>}
  n2 = LinearInterpolate(A=@n1.R, B=@n1.R, Alpha=ScalarParameter() {defaultValue=1, parameterName=Locked?})
  n3 = NamedRerouteUsage()
  Multiply(A=Max(A=Multiply(A=@n2, B=@n3), B=NamedRerouteUsage())) {constB=0.8}
  n4 = MaterialFunctionCall(Size=Constant() {r=0.15}, OutlineThickness=Constant() {r=0.0225}) {desc=circle, materialFunction=&MF_UI_Circle}
  n5 = MaterialFunctionCall(Size=Constant() {r=0.04}) {desc=dot, materialFunction=&MF_UI_Circle}
  NamedRerouteDeclaration(None=Max(
    A=NamedRerouteDeclaration(None=@n4.Outline) {name=InnerCircle, nodeColor=<rgba 1 0.3516 0 1>}
    B=NamedRerouteDeclaration(None=@n5.Fill) {name=Dot, nodeColor=<rgba 0.0312 0 1 1>}
   )) {name=InnerReticle, nodeColor=<rgba 1 0 0.6094 1>}
 ]
}
