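# Translucent UI material for a reticle overheat meter, drawing a segmented arc inside a tinted frame texture.
# An offset circle outline is cut by radial segment masks from NumberOfSegments and SegmentSpacing, then trimmed by radial progress.
# Overheat fills it in whole-segment steps via Ceil and blends the color from yellow toward red.
# CameraVelocity, clamped to a maximum drift, offsets the UVs for a parallax sway.
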Material /WiseOtter/Assets/UI/Reticle/M_UI_OverheatMeter
graph {
 properties {blendMode=BLEND_Translucent, materialDomain=MD_UI, pixelDepthOffsetMode=PDOM_Legacy, usedShadingModels=Default Lit}
 nodes [
  n1 = Constant3Vector() {constant=<rgba 1 0.8269 0.1669 1>}
  n2 = Constant3Vector() {constant=<rgba 1 0 0 0>}
  n3 = Saturate(None=Add(A=ScalarParameter() {defaultValue=0.666, parameterName=Overheat}) {constB=0.25})
  n4 = SmoothStep(Value=MaterialFunctionCall(
    UVs=NamedRerouteUsage()
    Size=Constant() {r=2}
    OutlineThickness=Constant() {r=0.0125}
    OutlineAlignment=Constant() {r=-5}
    GlowMax=Constant() {r=0.05}
   ) {materialFunction=&MF_UI_Circle}.Outline) {constMax=0.15}
  n5 = MaterialFunctionCall!(Rotation_Fraction=Constant() {r=0.75}, TexCoord=NamedRerouteUsage()) {materialFunction=/FortUILibrary/Foundation/Art_ToolKit/AngledBoxes/MF_UI_Rotator_Fractions}
  n6 = Constant() {r=1}
  n7 = ScalarParameter() {defaultValue=0.875, parameterName=SegmentSpacing}
  n8 = NamedRerouteUsage()
  n9 = Multiply(A=Fmod(A=@n8, B=Constant() {r=2})) {constB=0.5}
  n10 = Multiply(A=@n8) {constB=8}
  n11 = MaterialFunctionCall(Rotation_Fraction=Divide(A=Add(A=@n9) {constB=0.5}, B=@n10), TexCoord=NamedRerouteUsage()) {materialFunction=&MF_UI_Rotator_Fractions}
  n12 = Constant() {r=0.1}
  n13 = Multiply(
   A=SmoothStep(Value=MaterialFunctionCall!(UV=@n5.Result, Segment Count=@n6, Segment Spacing=@n7, Fixed Spacing) {materialFunction=/FortUILibrary/MaterialFunctions/Basics/MF_UI_RadialSegments}.Segmented Triangles Gradient) {constMax=0.015}
   B=SmoothStep(Value=MaterialFunctionCall(UV=@n11.Result, Segment Count=@n10, Segment Spacing=@n12) {materialFunction=&MF_UI_RadialSegments}.Segmented Triangles Gradient) {constMax=0.15}
  ) {constA=1}
  n14 = NamedRerouteUsage()
  n15 = NamedRerouteUsage()
  n16 = Multiply(A=ScalarParameter() {defaultValue=0.666, parameterName=Overheat}, B=@n15)
  n17 = OneMinus(None=MaterialFunctionCall(Input=Ceil(None=@n16), Input Low=Constant(), Input High=@n15) {materialFunction=&RemapValue_0-1}.Result)
  n18 = Constant() {r=67.5}
  n19 = Constant() {r=45}
  n20 = Saturate(None=Subtract(
    A=Multiply(A=@n4, B=@n13) {constA=1}
    B=MaterialFunctionCall!(UVs=@n14, Progress=@n17, StartValue=@n18, Range=@n19) {materialFunction=/FortUILibrary/MaterialFunctions/Basics/MF_UI_RadialProgress}.Result
   ))
  n21 = TextureSample(UVs=NamedRerouteUsage()) {texture=/WiseOtter/Assets/UI/Reticle/Textures/T_UI_OverheatFrame}
  Add(
   A=Multiply(A=LinearInterpolate(A=@n1, B=@n2, Alpha=@n3), B=@n20) {constA=1}
   B=Multiply(A=VectorParameter() {parameterName=FrameColor, defaultValue=<rgba 0.3968 0.4735 1 1>}.RGB, B=@n21.RGB) {constA=1}
  ) {constB=0}
  Add(A=@n21.RGB, B=@n20) {constB=0}
  NamedRerouteDeclaration(None=Ceil(None=ScalarParameter() {defaultValue=13, parameterName=NumberOfSegments})) {name="Ceil(NumberOfSegments)", nodeColor=<rgba 1 0 0.5391 1>}
  n22 = VectorParameter() {parameterName=CameraVelocity}
  n23 = Constant2Vector() {desc=Max Drift, g=0.07, r=0.35}
  n24 = Add(A=TextureCoordinate(), B=Clamp(None=ComponentMask(None=@n22.RGB) {g=true, r=true}, Min=Multiply(A=@n23) {constB=-1}, Max=@n23))
  NamedRerouteDeclaration(None=Subtract(A=NamedRerouteDeclaration(None=@n24) {name=ParallaxUV, nodeColor=<rgba 0 0 0 1>}, B=Constant2Vector() {r=-0.875})) {name=CircleUV, nodeColor=<rgba 0 0 0 1>}
 ]
}
