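# Reticle extension widget for the WiseOtter repulsors, showing lock-on, overheat and unibeam charge and cooldown.
# Repulsor weapon events drive the segmented overheat material and, via a delta gate, the lock-on ring animation that swaps with the center image.
# Async gameplay-tag waits on the viewed pawn play the charge-up animation and toggle cooldown, whose progress Tick feeds into the unibeam material.
# Tick also updates camera-rotation effects on the overheat meter.
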WidgetBlueprint /WiseOtter/Assets/UI/WBP_WiseOtter_Reticle
animations [
 Initalize {
  binding_count=2
  length_seconds=1
  object=&self:Initalize
  sequence {bindings [
    Image_Unibeam {
     id=#1
     possessed_class=Image
     tracks [
      {
       type=MovieSceneFloatTrack
       display_name=Render Opacity
       sections [{
         type=MovieSceneFloatSection
         end_frame=20
         start_frame=0
         channels [None {type=MovieSceneScriptingFloatChannel, keys [[0,0],[10,0],[14,0.5]]}]
        }]
      }
      {
       type=MovieScene2DTransformTrack
       sections [{
         type=MovieScene2DTransformSection
         end_frame=20
         start_frame=8
         channels [
          Translation.X {type=MovieSceneScriptingFloatChannel, const [0,1]}
          Translation.Y {type=MovieSceneScriptingFloatChannel, const [0,1]}
          Angle {type=MovieSceneScriptingFloatChannel, const [0,1]}
          Scale.X {type=MovieSceneScriptingFloatChannel, keys [[8,1],[10,1.25],[14,1]]}
          Scale.Y {type=MovieSceneScriptingFloatChannel, keys [[8,1],[10,1.05],[14,1]]}
          Shear.X {type=MovieSceneScriptingFloatChannel, const [0,1]}
          Shear.Y {type=MovieSceneScriptingFloatChannel, const [0,1]}
         ]
        }]
      }
     ]
    }
    Image_Center {
     id=#2
     possessed_class=Image
     tracks [{
       type=MovieScene2DTransformTrack
       sections [{
         type=MovieScene2DTransformSection
         end_frame=4
         start_frame=0
         channels [
          Translation.X {type=MovieSceneScriptingFloatChannel, const [0,1]}
          Translation.Y {type=MovieSceneScriptingFloatChannel, const [0,1]}
          Angle {type=MovieSceneScriptingFloatChannel, keys [[0,-30],[4,0]]}
          Scale.X {type=MovieSceneScriptingFloatChannel, keys [[0,1.5],[4,1]]}
          Scale.Y {type=MovieSceneScriptingFloatChannel, keys [[0,1.5],[4,1]]}
          Shear.X {type=MovieSceneScriptingFloatChannel, const [0,1]}
          Shear.Y {type=MovieSceneScriptingFloatChannel, const [0,1]}
         ]
        }]
      }]
    }
   ]}
 }
 Locked {
  binding_count=1
  length_seconds=0.25
  object=&self:Locked
  sequence {bindings [AutoAimTargetImage_LockOn {
     id=#3
     possessed_class=FortAutoAimTargetImage
     tracks [{
       type=MovieSceneWidgetMaterialTrack
       sections [{
         type=MovieSceneParameterSection
         end_frame=5
         start_frame=0
         channels [
          CircleMaskProgress {type=MovieSceneScriptingFloatChannel, keys [[0,0],[5,1]]}
          ColorR {type=MovieSceneScriptingFloatChannel, keys [[0,0.4062],[5,1]]}
          ColorG {type=MovieSceneScriptingFloatChannel, keys [[0,0.4841],[5,0]]}
          ColorB {type=MovieSceneScriptingFloatChannel, keys [[0,1],[5,0]]}
          ColorA {type=MovieSceneScriptingFloatChannel, const [1,2]}
         ]
        }]
      }]
    }]}
 }
 FireUnibeam {
  binding_count=2
  length_seconds=0.95
  object=&self:FireUnibeam
  sequence {bindings [
    Image_Center {
     id=#4
     possessed_class=Image
     tracks [{
       type=MovieSceneFloatTrack
       display_name=Render Opacity
       sections [{type=MovieSceneFloatSection, end_frame=14, start_frame=0, channels [None {type=MovieSceneScriptingFloatChannel, keys [[0,1],[14,0]]}]}]
      }]
    }
    AutoAimTargetImage_LockOn {
     id=#5
     possessed_class=FortAutoAimTargetImage
     tracks [
      {
       type=MovieScene2DTransformTrack
       sections [{
         type=MovieScene2DTransformSection
         end_frame=19
         start_frame=0
         channels [
          Translation.X {type=MovieSceneScriptingFloatChannel, const [0,1]}
          Translation.Y {type=MovieSceneScriptingFloatChannel, const [0,1]}
          Scale.X {type=MovieSceneScriptingFloatChannel, keys [[0,1.6],[19,1.2]]}
          Scale.Y {type=MovieSceneScriptingFloatChannel, keys [[0,1.6],[19,1.2]]}
          Shear.X {type=MovieSceneScriptingFloatChannel, const [0,1]}
          Shear.Y {type=MovieSceneScriptingFloatChannel, const [0,1]}
         ]
        }]
      }
      {
       type=MovieSceneFloatTrack
       display_name=Render Opacity
       sections [{type=MovieSceneFloatSection, channels [None {type=MovieSceneScriptingFloatChannel, keys [[0,0.1],[10,0]]}]}]
      }
     ]
    }
   ]}
 }
]
graphs [
 BindToWeaponEvents {nodes [
   n1 = BindToWeaponEvents!() -> (then, NewWeapon:Fort Weapon Object Reference) {is=title K2Node_FunctionEntry}
   > Branch(Condition="Not Equal (Object)"(A:Object Reference=@n1, B:Object Reference=Get Repulsors() -> (Repulsors:B High Tower Tomato Repulsors Object Reference)) -> (ReturnValue:bool=~) {is=title K2Node_PromotableOperator})
   > UnbindFromWeaponEvents()
   > n2 = Cast To B_HighTower_Tomato_Repulsors!(execute, Object=@n1) -> (then, CastFailed, AsB High Tower Tomato Repulsors:B High Tower Tomato Repulsors Object Reference) {is=title K2Node_DynamicCast}
   > n3 = Set Repulsors(Repulsors:B High Tower Tomato Repulsors Object Reference=@n2) -> (Output_Get:B High Tower Tomato Repulsors Object Reference)
   > Is Valid(InputObject=@n3)
   > Bind Event to On Overheat Value Changed(self:Fort Weapon Ranged Object Reference=@n3, Delegate=Create Event()) {is=title K2Node_AddDelegate}
   > Bind Event to Tomato Lock on Updated(self:B High Tower Tomato Repulsors Object Reference=@n3, Delegate=Create Event()) {is=title K2Node_AddDelegate}
   > InitializeOverheatValues()
   > OnTomatoLockOnUpdated!(
    execute
    self
    LockOnTarget:Actor Object Reference=Get LockOnTarget!(self:B High Tower Tomato Repulsors Object Reference=@n3) -> (LockOnTarget:Actor Object Reference)
   ) -> (then)
  ]}
 EventGraph {nodes [
   n1 = BindToPlayerEvents!() -> (OutputDelegate, then, NewPawn:Actor Object Reference) {is=title K2Node_CustomEvent}
   > n2 = Cast To FortPlayerPawn(Object=@n1.NewPawn)
   > Set Pawn(Pawn:Fort Player Pawn Object Reference=@n2) -> (Output_Get:Fort Player Pawn Object Reference)
   > n3 = WaitGameplayTagAddToActor!(
    execute:Exec
    TargetActor:Actor Object Reference=Get Pawn() -> (Pawn:Fort Player Pawn Object Reference)
    Tag:Gameplay Tag Structure=Get ChargingUpTag() -> (ChargingUpTag:Gameplay Tag Structure)
    OnlyTriggerOnce:bool=false
   ) -> (then:Exec, AsyncTaskProxy:Ability Async Wait Gameplay Tag Added Object Reference, Added:Exec) {is=title K2Node_AsyncAction}
   > n4 = WaitGameplayTagRemoveFromActor!(execute:Exec, TargetActor:Actor Object Reference=Get Pawn(), Tag:Gameplay Tag Structure=Get ChargingUpTag(), OnlyTriggerOnce:bool=false) -> (then:Exec, AsyncTaskProxy:Ability Async Wait Gameplay Tag Removed Object Reference, Removed:Exec) {is=title K2Node_AsyncAction}
   > n5 = WaitGameplayTagAddToActor(TargetActor=Get Pawn(), Tag=(TagName="Granted.Cooldown.HighTower.Tomato.RepulsorCannon"), OnlyTriggerOnce=false)
   > n6 = WaitGameplayTagRemoveFromActor(TargetActor=Get Pawn(), Tag=(TagName="Granted.Cooldown.HighTower.Tomato.RepulsorCannon"), OnlyTriggerOnce=false)
   SetScalarParameterValue(
    execute=@n6.Removed
    self=Get UnibeamMaterial() -> (UnibeamMaterial:Material Instance Dynamic Object Reference)
    ParameterName=Ready?
    Value=1
   )
   > SetScalarParameterValue(self=Get UnibeamMaterial(), ParameterName=CooldownProgress, Value=1)
   > Set IsCoolingDownUnibeam(IsCoolingDownUnibeam:bool=false) -> (Output_Get:bool)
   Set IsCoolingDownUnibeam(execute=@n5.Added, IsCoolingDownUnibeam=true)
   PlayAnimationReverse(execute=@n4.Removed, InAnimation=Get FireUnibeam() -> (FireUnibeam:Widget Animation Object Reference))
   PlayAnimationForward(execute=@n3.Added, InAnimation=Get FireUnibeam())
   Event OnInitialized()
   > Parent: OnInitialized() {is=title K2Node_CallParentFunction}
   > n7 = GetDynamicMaterial(self=Get Image_Unibeam() -> (Image_Unibeam:Image Object Reference))
   > Set UnibeamMaterial(UnibeamMaterial:Material Instance Dynamic Object Reference=@n7) -> (Output_Get:Material Instance Dynamic Object Reference)
   > n8 = GetDynamicMaterial(self=Get Image_Overheat() -> (Image_Overheat:Image Object Reference))
   > Set OverheatMaterial(OverheatMaterial:Material Instance Dynamic Object Reference=@n8) -> (Output_Get:Material Instance Dynamic Object Reference)
   Event Construct()
   > BindToWeaponEvents!(
    execute
    self
    NewWeapon:Fort Weapon Object Reference=GetEquippedWeapon!(self:Fort Weapon Reticle Extension Widget Base Object Reference) -> (ReturnValue:Fort Weapon Object Reference)
   ) -> (then) {is=title K2Node_CallFunction}
   > BindToPlayerEvents!(
    execute
    self
    NewPawn:Actor Object Reference=GetViewTarget!(self:Controller Object Reference=GetOwningPlayer()) -> (ReturnValue:Actor Object Reference)
   ) -> (then) {is=title K2Node_CallFunction}
   > n9 = Create UICamera Effects from Image List!(
    execute:Exec
    ImageList:Array of Image Object References=Make Array("[0]":Image Object Reference=Get Image_Overheat()) -> (Array:Array of Image Object References)
    PitchMaterial:Material Instance Dynamic Object Reference
   ) -> (then:Exec, CameraEffectsObject:"Fort Widget Camera Rotation Effects Object Reference (by ref)") {is=title K2Node_MacroInstance}
   > Set CameraRotationEffects(CameraRotationEffects:Fort Widget Camera Rotation Effects Object Reference=@n9) -> (Output_Get:Fort Widget Camera Rotation Effects Object Reference)
   > PlayAnimationForward(InAnimation=Get Initalize() -> (Initalize:Widget Animation Object Reference))
   n10 = Event Tick()
   > Update!(
    execute
    self:Fort Widget Camera Rotation Effects Object Reference=Get CameraRotationEffects() -> (CameraRotationEffects:Fort Widget Camera Rotation Effects Object Reference)
    DeltaTime:float=@n10.InDeltaTime
   ) -> (then)
   > Branch(Condition=Get IsCoolingDownUnibeam() -> (IsCoolingDownUnibeam:bool))
   > n11 = GetCooldownTimeRemainingAndDuration!(
    execute
    Actor:Actor Object Reference=Get Pawn()
    CooldownTags:Gameplay Tag Container Structure=(GameplayTags=((TagName="Granted.Cooldown.HighTower.Tomato.RepulsorCannon")))
   ) -> (then, TimeRemaining:float, CooldownDuration:float)
   > SetScalarParameterValue(self=Get UnibeamMaterial(), ParameterName=CooldownProgress, Value=SafeDivide(A=@n11, B=@n11.CooldownDuration))
   > SetScalarParameterValue(self=Get UnibeamMaterial(), ParameterName=Ready?, Value=0)
   Event Destruct()
   > UnbindFromWeaponEvents()
  ]}
 InitializeOverheatValues {nodes [
   InitializeOverheatValues() {is=title K2Node_FunctionEntry}
   n1 = Get Repulsors()
   n2 = GetOverheatingMaxValue!(self:Fort Weapon Ranged Object Reference=@n1) -> (ReturnValue:float)
   > Set OverheatingMaxValue(OverheatingMaxValue:float=@n2) -> (Output_Get:float)
   > SetScalarParameterValue(
    self=Get OverheatMaterial() -> (OverheatMaterial:Material Instance Dynamic Object Reference)
    ParameterName=NumberOfSegments
    Value=float / float(A=@n2, B=GetOverheatHeatingValue!(self:Fort Weapon Ranged Object Reference=@n1) -> (ReturnValue:float))
   )
   > OnOverheatValueChanged!(
    execute
    self
    CurrentValue:float=GetCurrentOverheatValue!(self:Fort Weapon Ranged Object Reference=Get Repulsors()) -> (ReturnValue:float)
    MaxValue:float=Get OverheatingMaxValue() -> (OverheatingMaxValue:float)
    State:EFortWeaponOverheatState Enum=None
    bIsInCooldown:bool=false
   ) -> (then)
  ]}
 OnOverheatValueChanged {nodes [
   n1 = OnOverheatValueChanged!() -> (then, CurrentValue:float=~, MaxValue:float=~, State:EFortWeaponOverheatState Enum=None, bIsInCooldown:bool=~) {is=title K2Node_FunctionEntry}
   > n2 = Set NormalizedOverheat(NormalizedOverheat:double=float / float(A=@n1, B=@n1.MaxValue)) -> (Output_Get:double)
   > SetScalarParameterValue(self=Get OverheatMaterial(), ParameterName=Overheat, Value=@n2)
   n3 = SelectFloat(A=0.25, B=0, bPickA="float > float"(A=@n2))
   > SetRenderOpacity(self=Get Image_Overheat(), InOpacity=MapRangeClamped(Value=float + float(A=@n2, B=@n3), InRangeA=0, InRangeB=0.75))
  ]}
 OnTomatoLockOnUpdated {nodes [
   n1 = OnTomatoLockOnUpdated!() -> (then, LockOnTarget:Actor Object Reference) {is=title K2Node_FunctionEntry}
   > n2 = Delta Gate Wildcard!(execute:Exec, Variable:"Boolean (by ref)"=Get HasValidTarget() -> (HasValidTarget:bool), New Value:bool=IsValid(Object=@n1)) -> (Success:Exec, Updated Variable:"Boolean (by ref)", Previous Variable:bool=~) {is=title K2Node_MacroInstance}
   > n3 = Branch(Condition=@n2)
   n4 = Get Image_Center() -> (Image_Center:Image Object Reference)
   > SetVisibility(self=@n4, InVisibility=Hidden)
   n5 = Get Locked() -> (Locked:Widget Animation Object Reference)
   > PlayAnimationForward(InAnimation=@n5)
   SetVisibility(execute=@n3.else, self=@n4, InVisibility=SelfHitTestInvisible)
   > PlayAnimationReverse(InAnimation=@n5)
  ]}
 UnbindFromWeaponEvents {nodes [
   UnbindFromWeaponEvents() {is=title K2Node_FunctionEntry}
   n1 = Get Repulsors()
   > Is Valid(InputObject=@n1)
   > Unbind Event from On Overheat Value Changed(self:Fort Weapon Ranged Object Reference=@n1, Delegate=Create Event()) {is=title K2Node_RemoveDelegate}
   > Unbind Event from Tomato Lock on Updated(self:B High Tower Tomato Repulsors Object Reference=@n1, Delegate=Create Event()) {is=title K2Node_RemoveDelegate}
   > Set Repulsors()
  ]}
]
variables [
 CameraRotationEffects {type=FortWidgetCameraRotationEffects, default=null}
 ChargingUpTag {type=GameplayTag, default {tagName=GameplayCue.HighTower.Tomato.RepulsorCannon.ChargeUp}}
 HasValidTarget {type=boolean}
 IsCoolingDownUnibeam {type=boolean}
 NormalizedOverheat {type=number}
 Out Target Pawn {type=FortPawn, default=null}
 OverheatMaterial {type=MaterialInstanceDynamic}
 Pawn {type=FortPlayerPawn}
 Repulsors {type=B_HighTower_Tomato_Repulsors_C, default=null}
 TargetStateComponent {type=FortWeaponComponent_TargetState, default=null}
 UnibeamMaterial {type=MaterialInstanceDynamic}
]
widgets {
 info {
  inheritedWidgetCount=1
  parentClass=/Script/FortniteUI.FortWeaponReticleExtensionWidgetBase
  rootWidgetClass=&Overlay
  widgetCount=6
 }
 widgets [
  Overlay_18 {widgetClassPath=&Overlay, widget=Overlay_18}
  AutoAimTargetImage_LockOn {
   widgetClassPath=/Script/AutoAimWeaponUI.FortAutoAimTargetImage
   bIsVariable=true
   parent=Overlay_18
   widget=AutoAimTargetImage_LockOn
   properties {
    visibility=SelfHitTestInvisible
    autoAimReticleSize {
     registryType=None
     value=1
     curve {curveTable=/WiseOtter/DataTables/WiseOtterGameData, rowName=Default.WiseOtter.Repulsors.LockOn.ReticleSize}
    }
    brush {
     bIsDynamicallyLoaded=false
     drawAs=Image
     imageType=FullColor
     mirroring=NoMirror
     resourceName=None
     resourceObject=/WiseOtter/Assets/UI/Reticle/M_LockOnRing
     tiling=NoTile
     imageSize=<256 256>
     margin=<ltrb 0 0 0 0>
     outlineSettings {
      bUseBrushTransparency=false
      roundingType=HalfHeightRadius
      width=0
      color {colorUseRule=UseColor_Specified, specifiedColor=<rgba 0 0 0 0>}
      cornerRadii=<0 0 0 0>
     }
     tintColor {colorUseRule=UseColor_Specified, specifiedColor=<rgba 1 1 1 1>}
     uVRegion {bIsValid=false, max=<0 0>, min=<0 0>}
    }
   }
   slot {class=OverlaySlot, properties {horizontalAlignment=HAlign_Center, verticalAlignment=VAlign_Center}}
  }
  Image_Unibeam {
   widgetClassPath=&Image
   bIsVariable=true
   parent=Overlay_18
   widget=Image_Unibeam
   properties {
    visibility=SelfHitTestInvisible
    brush {
     bIsDynamicallyLoaded=false
     drawAs=Image
     imageType=FullColor
     mirroring=NoMirror
     resourceName=None
     resourceObject=/WiseOtter/Assets/UI/Reticle/M_triangle
     tiling=NoTile
     imageSize=<256 256>
     margin=<ltrb 0 0 0 0>
     outlineSettings {
      bUseBrushTransparency=false
      roundingType=HalfHeightRadius
      width=0
      color {colorUseRule=UseColor_Specified, specifiedColor=<rgba 0 0 0 0>}
      cornerRadii=<0 0 0 0>
     }
     tintColor {colorUseRule=UseColor_Specified, specifiedColor=<rgba 1 1 1 1>}
     uVRegion {bIsValid=false, max=<0 0>, min=<0 0>}
    }
   }
   slot {class=OverlaySlot, properties {horizontalAlignment=HAlign_Center, verticalAlignment=VAlign_Center}}
  }
  Image_Center {
   widgetClassPath=&Image
   bIsVariable=true
   parent=Overlay_18
   widget=Image_Center
   properties {
    visibility=SelfHitTestInvisible
    brush {
     bIsDynamicallyLoaded=false
     drawAs=Image
     imageType=FullColor
     mirroring=NoMirror
     resourceName=None
     resourceObject=/WiseOtter/Assets/UI/Reticle/M_RepulsorCenter
     tiling=NoTile
     imageSize=<90 90>
     margin=<ltrb 0 0 0 0>
     outlineSettings {
      bUseBrushTransparency=false
      roundingType=HalfHeightRadius
      width=0
      color {colorUseRule=UseColor_Specified, specifiedColor=<rgba 0 0 0 0>}
      cornerRadii=<0 0 0 0>
     }
     tintColor {colorUseRule=UseColor_Specified, specifiedColor=<rgba 1 1 1 1>}
     uVRegion {bIsValid=false, max=<0 0>, min=<0 0>}
    }
   }
   slot {class=OverlaySlot, properties {horizontalAlignment=HAlign_Center, verticalAlignment=VAlign_Center}}
  }
  Image_Overheat {
   widgetClassPath=&Image
   bIsVariable=true
   parent=Overlay_18
   widget=Image_Overheat
   properties {
    renderOpacity=0
    visibility=SelfHitTestInvisible
    brush {
     bIsDynamicallyLoaded=false
     drawAs=Image
     imageType=FullColor
     mirroring=NoMirror
     resourceName=None
     resourceObject=/WiseOtter/Assets/UI/Reticle/M_UI_OverheatMeter
     tiling=NoTile
     imageSize=<128 140>
     margin=<ltrb 0 0 0 0>
     outlineSettings {
      bUseBrushTransparency=false
      roundingType=HalfHeightRadius
      width=0
      color {colorUseRule=UseColor_Specified, specifiedColor=<rgba 0 0 0 0>}
      cornerRadii=<0 0 0 0>
     }
     tintColor {colorUseRule=UseColor_Specified, specifiedColor=<rgba 1 1 1 1>}
     uVRegion {bIsValid=false, max=<0 0>, min=<0 0>}
    }
    renderTransform {angle=0, scale=<1 1>, shear=<0 0>, translation=<130 0>}
   }
   slot {class=OverlaySlot, properties {horizontalAlignment=HAlign_Center, verticalAlignment=VAlign_Center}}
  }
  TutorialHighlight {widgetClassPath=&TutorialHighlight}
 ]
}
